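# Looping sound cue for the "cook" stage of the DogSweater item, routed through the explosives sound class.
# After a short fixed delay, an envelope ramps volume up from near silence over a quieting modulator.
# A quality switch picks either the standard loop or an optimized mobile loop, the latter with a gain boost.
# Overridden attenuation keeps it short-range with natural falloff, and overridden concurrency prevents new instances at the limit.
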SoundCue /FNGameplayCues/Athena/Sounds/Items/DogSweater/Athena_DogSweater_Cook_Loop_Cue
root=Delay_0
nodes [
 {id=Delay_0, children [Enveloper_0], properties {delayMax=0.3, delayMin=0.3}}
 {
  id=Enveloper_0
  children [Modulator_0]
  properties {pitchCurve {editorCurveData {keys [{time=0}]}}, volumeCurve {editorCurveData {keys [<at 0 0.05>,<at 1 1>]}}}
 }
 {id=Modulator_0, children [QualityLevel_0], properties {pitchMax=1, pitchMin=1, volumeMax=0.2, volumeMin=0.2}}
 {id=Modulator_1, children [WavePlayer_0], properties {pitchMax=1, pitchMin=1, volumeMax=2, volumeMin=2}}
 {id=QualityLevel_0, children [WavePlayer_1,WavePlayer_1,Modulator_1,Modulator_1]}
 {
  id=WavePlayer_0
  properties {bLooping=true, soundWaveAssetPtr=/FNGameplayCues/Athena/Sounds/Items/DogSweater/OPT/Mobile/Athena_DogSweater_Cook_Loop_OPT}
 }
 {id=WavePlayer_1, properties {bLooping=true, soundWaveAssetPtr=&/Athena_DogSweater_Cook_Loop}}
]
properties {
 bOverrideAttenuation=true
 bOverrideConcurrency=true
 maxDistance=200
 soundClassObject=/Game/Sounds/FortSoundClassesAndModes/Fort_Explosives
 volumeMultiplier=0.15
 attenuationOverrides {distanceAlgorithm=NaturalSound, falloffDistance=100, stereoSpread=0, attenuationShapeExtents {x=100}}
 concurrencyOverrides {resolutionRule=PreventNew}
}
