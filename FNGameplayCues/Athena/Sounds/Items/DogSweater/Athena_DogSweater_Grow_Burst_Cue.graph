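# Sound cue for the DogSweater item's grow burst effect, in the Fort gadgets/items sound class.
# A quality switch plays one of two random burst waves on the first two levels, or a single volume-boosted mobile-optimized wave on the other two.
# The result runs through a short pitch and volume envelope with a fast fade-in, then a root modulator.
# Attenuation and concurrency are overridden on the cue, using natural-sound falloff and a two-instance limit.
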SoundCue /FNGameplayCues/Athena/Sounds/Items/DogSweater/Athena_DogSweater_Grow_Burst_Cue
root=Modulator_0
nodes [
 {
  id=Enveloper_0
  children [QualityLevel_0]
  properties {pitchCurve {editorCurveData {keys [<at 0 0.5>, {time=0.15}]}}, volumeCurve {editorCurveData {keys [<at 0 0.01>,<at 0.1 1>]}}}
 }
 {id=Modulator_0, children [Enveloper_0]}
 {id=Modulator_1, children [WavePlayer_0], properties {pitchMax=1, pitchMin=1, volumeMax=2, volumeMin=2}}
 {id=QualityLevel_0, children [Random_0,Random_0,Modulator_1,Modulator_1]}
 {id=Random_0, children [WavePlayer_1,WavePlayer_2], properties {weights [1,1]}}
 {
  id=WavePlayer_0
  properties {soundWaveAssetPtr=/FNGameplayCues/Athena/Sounds/Items/DogSweater/OPT/Mobile/Athena_DogSweater_Grow_Burst_01_OPT}
 }
 {id=WavePlayer_1, properties {soundWaveAssetPtr=&/Athena_DogSweater_Grow_Burst_01}}
 {id=WavePlayer_2, properties {soundWaveAssetPtr=&/Athena_DogSweater_Grow_Burst_02}}
]
properties {
 bOverrideAttenuation=true
 bOverrideConcurrency=true
 maxDistance=9500
 soundClassObject=&Fort_Gadgets_Items
 volumeMultiplier=0.7
 attenuationOverrides {distanceAlgorithm=NaturalSound, falloffDistance=9000, attenuationShapeExtents {x=500}}
 concurrencyOverrides {maxCount=2, volumeScale=0.5, volumeScaleAttackTime=1}
 concurrencySet [null]
}
