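# Sound cue for the DogSweater projectile impact, layering an impact sound with a charge-up sound.
# A quality-level node sends its first two slots to a mix of two randomized variants per layer, and its last two to a mix of one fixed wave per layer.
# Each mix boosts the charge-up input, while a modulator varies the impact volume.
# The cue overrides attenuation with natural-sound falloff and caps concurrency at three instances.
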SoundCue /FNGameplayCues/Athena/Sounds/Items/DogSweater/Athena_DogSweater_Projectile_Impact_Cue
root=QualityLevel_0
nodes [
 {id=Mixer_0, children [Modulator_0,Random_1], properties {inputVolume [0.75,1.5]}}
 {id=Mixer_2, children [Modulator_5,WavePlayer_7], properties {inputVolume [0.75,1.5]}}
 {id=Modulator_0, children [Random_0], properties {volumeMax=1.6, volumeMin=1.5}}
 {id=Modulator_5, children [WavePlayer_6], properties {volumeMax=1.6, volumeMin=1.5}}
 {id=QualityLevel_0, children [Mixer_0,Mixer_0,Mixer_2,Mixer_2]}
 {id=Random_0, children [WavePlayer_1,WavePlayer_2], properties {weights [1,1]}}
 {id=Random_1, children [WavePlayer_3,WavePlayer_4], properties {weights [1,1]}}
 {id=WavePlayer_1, properties {soundWaveAssetPtr=&/Athena_DogSweater_Projectile_Impact_01}}
 {id=WavePlayer_2, properties {soundWaveAssetPtr=&/Athena_DogSweater_Projectile_Impact_02}}
 {id=WavePlayer_3, properties {soundWaveAssetPtr=&/Athena_DogSweater_ChargeUp_01}}
 {id=WavePlayer_4, properties {soundWaveAssetPtr=&/Athena_DogSweater_ChargeUp_02}}
 {id=WavePlayer_6, properties {soundWaveAssetPtr=&Athena_DogSweater_Projectile_Impact_01}}
 {id=WavePlayer_7, properties {soundWaveAssetPtr=&Athena_DogSweater_ChargeUp_02}}
]
properties {
 bOverrideAttenuation=true
 bOverrideConcurrency=true
 maxDistance=7000
 soundClassObject=&Fort_WeaponRanged_Fire_3P_PSM
 volumeMultiplier=0.7
 attenuationOverrides {distanceAlgorithm=NaturalSound, falloffDistance=6000, attenuationShapeExtents {x=1000}}
 concurrencyOverrides {maxCount=3, volumeScaleAttackTime=0.3}
 concurrencySet [null]
}
